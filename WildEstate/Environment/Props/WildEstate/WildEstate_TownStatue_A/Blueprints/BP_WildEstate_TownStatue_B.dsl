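; Destructible Wild Estate town statue that loses its head when damaged and is replaced by an indestructible statue actor.
; On a server without cosmetic events, damage or death with a valid causer caches instigator, causer and momentum, fires a burst gameplay cue and silently kills the statue.
; The server death event spawns the replacement, hands it the cached damage data and sends the instigator a head-detached quest event.
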(event EventBeginPlay
  (bind _outputdelegate (EventDispatchers|CreateEvent :!event Custom|OnHeadDetachTriggered))
  (Class|FortAsyncActionCurrentPlaylistReady|PlaylistDataReadyAsync
    (:OnReady
      (Networking|CanExecuteCosmeticEvents
        (:False
          (if (Networking|IsServer)
            ;[0 0 0 0.5] Bind callback to both OnDamaged and OnDied to account for cases where the statue is destroyed without a damage event
            (EventDispatchers|AssignOnDamaged _outputdelegate)
            (EventDispatchers|BindEventtoOnDied _outputdelegate)))))))

(event Custom|OnHeadDetachTriggered
  (params (DamagedActor :type Actor%20Object%20Reference) (Damage :type Float%20%28single-precision%29) (InstigatedBy :type Controller%20Object%20Reference) (DamageCauser :type Actor%20Object%20Reference) (HitLocation :type Vector) (FHitComponent :type Primitive%20Component%20Object%20Reference) (BoneName :type Name) (Momentum :type Vector))
  (bind _outputdelegate (EventDispatchers|CreateEvent :!event Custom|OnHeadDetachTriggered))
  (Utilities|IsValid DamageCauser
    (:Is%20Valid
      (bind _output_get (Variables|Default|SetDetachInstigator InstigatedBy))
      (bind _output_get_2 (Variables|Default|SetDetachDamageCauser DamageCauser))
      (Variables|Default|SetDetachMomentum Momentum)
      (GameplayCue|ExecuteGameplayCueOnActor(Burst) self (Variables|Default|GetGCN_DamagedEvent) (Ability|GameplayCue|MakeGameplayCueParameters :NormalizedMagnitude 0.0 :RawMagnitude 0.0 :EffectContext (Class|AthenaFunctionLibrary|MakeSimpleFortEffectContext _output_get _output_get_2) :Location HitLocation :Normal "0,0,0" :Instigator _output_get :EffectCauser _output_get_2 :GameplayEffectLevel 1 :AbilityLevel 1 :bReplicateLocationWhenUsingMinimalRepProxy true))
      ;[0 0 0 0.5] Destroy statue immediately when head is removed if it's not already being destroyed
      ; - we don't want the statue to still take damage while the head is already falling off
      (if (not (Health|WasDestroyed))
        (Gameplay|SilentDie false)
        ;[0 0 0 0.5] Unbind OnHeadDetachTriggered from OnDamaged and OnDied once the invulnerable statue is spawned
        (EventDispatchers|UnbindEventfromOnDamaged _outputdelegate)
        (EventDispatchers|UnbindEventfromOnDied _outputdelegate)))))

(event EventOnDeathServer
  (params (Damage :type Float%20%28single-precision%29) (DamageTags :type Gameplay%20Tag%20Container%20Structure%20%28by%20ref%29) (Momentum :type Vector :default "0, 0, 0") (HitInfo :type Hit%20Result%20Structure%20%28by%20ref%29) (InstigatedBy :type Controller%20Object%20Reference) (DamageCauser :type Actor%20Object%20Reference) (EffectContext :type Gameplay%20Effect%20Context%20Handle%20Structure))
  ;[0 0 0 0.5] Spawn indestructible building actor after this actor has been destroyed, similar to Prop_PhysicsFellable_Parent
  (bind _returnvalue (Game|SpawnBuildingActorfromClass (Variables|Default|GetIndestructibleStatueBaseClass) (Transformation|GetWorldTransform (Variables|Debug|GetStaticMeshComponent)) 0 false 0))
  (bind _asbp_wild_estate_town_statue_indestructible_base (Utilities|Casting|CastToBP_WildEstate_TownStatue_IndestructibleBase _returnvalue))
  (Class|BPWildEstateTownStatueIndestructibleBase|InitDamageVariables _asbp_wild_estate_town_statue_indestructible_base (Variables|Default|GetDetachInstigator) (Variables|Default|GetDetachDamageCauser) (Variables|Default|GetDetachMomentum))
  ;[0.03135 0.01734 0.15] Send custom quest event since this actor is usually destroyed by itself after taking some damage. It doesn't have a reliable death event
  (Ability|SendGameplayEventtoActor (Variables|Default|GetDetachInstigator) (Variables|Default|GetQuest_HeadDetachedEventTag) (Utilities|Struct|MakeGameplayEventData (Variables|Default|GetQuest_HeadDetachedEventTag) (Variables|Default|GetDetachInstigator) self (Variables|Default|GetDetachDamageCauser) 0 "()" (GameplayTags|GetOwnedGameplayTags (Ability|GetFortAbilitySystemComponent (Variables|Default|GetDetachInstigator)) :!class /Script/GameplayTags.BlueprintGameplayTagLibrary))))

(fn ConstructionScript
  (|CallParentFunction))
