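; Indestructible base of the WildEstate town statue that plays the head toppling and breaking off when it begins play.
; The server polls briefly for the damage instigator, then picks the fall direction from that pawn's side of the statue, defaulting to forward.
; A gameplay cue fires and a timeline tilts the head pivot.
; The server then spawns an impulse-launched BondShock physics head, while bHeadDetached's OnRep hides the original mesh.
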(event EventBeginPlay
  ;[0 1 0.24] BeginPlay
  (|CallParentFunction)
  (Utilities|FlowControl|DoOnce false
    (:Completed
      (Networking|SwitchHasAuthority
        (:Authority
          (CallFunction|PollInstigatorValidity))))))

(event Custom|PollInstigatorValidity
  ;[0.1659 0.4306 0.5747] PollInstigatorValidity
  ; Try to wait for the damaging instigator in order to play the head animations in the expected direction
  (Variables|Default|GetStatueDamageInstigator
    (:then
      _)
    (:else
      (if (not (>= (Variables|Default|GetCurrentInstigatorChecks) (Variables|Default|GetMaxInstigatorChecks)))
        (Utilities|Time|SetTimerbyEvent (EventDispatchers|CreateEvent :!event Custom|PollInstigatorValidity) 0.2)
        (Math|Integer|IncrementInt (Variables|Default|GetCurrentInstigatorChecks)))))
  (CallFunction|StartHeadFall))

(event Custom|StartHeadFall
  (Utilities|FlowControl|Sequence
    (:then_0
      (Networking|SwitchHasAuthority
        (:Authority
          (CallFunction|FindHeadFallDirection) ; finds and stores the head fall direction
        )))
    (:then_1
      (GameplayCue|ExecuteGameplayCueOnActor(Burst) self (Variables|Default|GetGCN_DetachHead) (Ability|GameplayCue|MakeGameplayCueParameters :NormalizedMagnitude 0.0 :RawMagnitude 0.0 :EffectContext (Class|AthenaFunctionLibrary|MakeSimpleFortEffectContext self self) :Location (.location (Transformation|GetWorldTransform (Variables|Default|GetHeadMesh))) :Normal "0,0,0" :GameplayEffectLevel 1 :AbilityLevel 1 :bReplicateLocationWhenUsingMinimalRepProxy false)))
    (:then_2
      (bind (_direction _headrot) (Timeline|Timeline_HeadFall Play)
        (:Update
          (Transformation|SetRelativeRotation (Variables|Default|GetHeadPivot) (Math|Rotator|MakeRotator (Math|Float|SelectFloat (* _headrot -1.0) _headrot (Variables|Default|GetbHeadWillFallBackward)) (.pitch (Variables|/Script/Engine.SceneComponent|GetRelativeRotation (Variables|Default|GetHeadPivot))) (.yaw (Variables|/Script/Engine.SceneComponent|GetRelativeRotation (Variables|Default|GetHeadPivot)))) false false))
        (:Finished
          (CallFunction|DelayedDetachHead))))))

(event Custom|WakeHeadPhysics
  (params (ImpulseDirection :type Vector))
  (bind _head (Variables|Default|GetHead)
    (:then
      (Class|BPBondShockPhysics|WakeandAddImpulse _head (* ImpulseDirection (Variables|Default|GetDecapitationMagnitude))))
    (:else)))

(event Custom|DelayedDetachHead
  (if (Networking|IsServer)
    (bind _returnvalue (Game|SpawnBuildingActorfromClass "/WildEstate/Gameplay/BondShock/BP_BondShock_Physics.BP_BondShock_Physics_C" (Transformation|GetWorldTransform (Variables|Default|GetHeadMesh)) 0 false false false 0 0 0))
    (Variables|Default|SetHead _returnvalue)
    (Variables|Default|SetbHeadDetached true)
    (CallFunction|WakeHeadPhysics (Math|Vector|SelectVector (Transformation|GetActorRightVector) (* (Transformation|GetActorRightVector) -1.0) (Variables|Default|GetbHeadWillFallBackward)))))

(fn FindHeadFallDirection
  (bind _statuedamageinstigator (Variables|Default|GetStatueDamageInstigator))
  (Utilities|IsValid _statuedamageinstigator
    (:Is%20Valid
      (bind _output_get (Variables|Default|SetbHeadWillFallBackward (>= (Math|Vector|DotProduct (Transformation|GetActorRightVector self) (Math|Vector|Normalize (- (Transformation|GetActorLocation (Pawn|GetControlledPawn _statuedamageinstigator)) (Transformation|GetActorLocation self)))) 0)))
      (return _output_get))
    (:Is%20Not%20Valid
      (bind _output_get_2 (Variables|Default|SetbHeadWillFallBackward false)) ; use falling forward as the default value
      (return _output_get_2))))

(fn InitDamageVariables
  (params (DamageInstigator :type Controller%20Object%20Reference) (DamageSource :type Actor%20Object%20Reference) (DetachMomentum :type Vector))
  (Variables|Default|SetStatueDamageInstigator DamageInstigator)
  (Variables|Default|SetStatueDamageSource DamageSource)
  (Variables|Default|SetDetachMomentum DetachMomentum))

(fn OnRep_bHeadDetached
  (bind _headmesh (Variables|Default|GetHeadMesh))
  (Development|SetHiddeninGame _headmesh true false)
  (Collision|SetCollisionEnabled _headmesh "NoCollision"))

(fn SpawnDetachedHead)

(fn ConstructionScript
  (|CallParentFunction))
