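; Parent Blueprint for a WildEstates physics ball that players and other balls knock around, using projectile movement rather than physics simulation.
; Overlaps apply clamped, angle-scaled impulses, while air vents and StartMovement can also launch it.
; Ticking and net dormancy follow motion: the server stops slow balls in water, while clients finish mesh interpolation before sleeping.
; Hits and overlaps drive collision audio and bounce-limited dust or water effects.
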(event EventBeginPlay
  (EventDispatchers|BindEventtoOnProjectileStop (Variables|Default|GetProjectileMovement) (delegate Custom|ProjectileStop))
  (bind _output_get (Variables|/Script/Engine.MovementComponent|SetVelocity (Transformation|GetVelocity) (Variables|Default|GetProjectileMovement))) ; make sure the movement comp velocity\nmatches the root comp velocity
  (CallFunction|SetTickingEnabled (Math|Vector|NotEqual(Vector) _output_get "0, 0, 0" 0.0001)) ; start ticking only if we already have a velocity
  (Variables|Default|SetInitialSpawnPosition (Transformation|GetActorLocation))
  (Networking|CanExecuteCosmeticEvents
    (:True
      (Variables|FortPhysicsBall|SetRotatedComponent (Variables|Default|GetMesh))
      (Water|BindEventtoOnEnterWaterEvent (Variables|FortPhysicsBall|GetWaterInteractionComponent) (EventDispatchers|CreateEvent :!event Custom|Water%20Collision)))))

(event Custom|DrawDebugTrajectory)

(event Custom|ProjectileStop
  (params (ImpactResult :type Hit%20Result%20Structure%20%28by%20ref%29))
  (CallFunction|UpdateDirectionandRotationRate 0.0)
  (Variables|Default|SetEverCameToRest true)
  (Networking|CanExecuteCosmeticEvents
    (:False
      (CallFunction|SetTickingEnabled false) ; Only on server. Client waits for interp in tick.
    )))

(event Movement|EventOnReplicatedVelocityStartOrStop
  (params (bStart :type Boolean))
  (if bStart
    (CallFunction|SetTickingEnabled true)
    (CallFunction|UpdateDirectionandRotationRate 0.0)
    (else
      (CallFunction|UpdateDirectionandRotationRate 0.0)
      (Variables|Default|SetEverCameToRest true)
      (Networking|CanExecuteCosmeticEvents
        (:False
          (CallFunction|SetTickingEnabled false) ; Only on server. Client waits for interp in tick.
        )))))

(event EventTick
  (params (DeltaSeconds :type Float%20%28single-precision%29))
  (|CallParentFunction DeltaSeconds)
  (Networking|CanExecuteCosmeticEvents
    (:True
      (CallFunction|UpdateDirectionandRotationRate DeltaSeconds)
      (Utilities|FlowControl|Sequence
        (:then_0
          (if (Variables|Config|GetPlotTrajectory)
            (CallFunction|DrawDebugTrajectory)))
        (:then_1
          (if (and (Game|Components|ProjectileMovement|IsVelocityUnderSimulationThreshold (Variables|Default|GetProjectileMovement)) (Game|Components|ProjectileMovement|Interpolation|IsInterpolationComplete (Variables|Default|GetProjectileMovement)))
            (CallFunction|SetTickingEnabled false)
            (Game|Components|ProjectileMovement|StopSimulating (Variables|Default|GetProjectileMovement) (Collision|MakeHitResult false))))))
    (:False
      (CallFunction|CheckStopinWateronServer))))

(event OnComponentBeginOverlap(SphereCollision)
  (params (OverlappedComponent :type Primitive%20Component%20Object%20Reference) (OtherActor :type Actor%20Object%20Reference) (OtherComp :type Primitive%20Component%20Object%20Reference) (OtherBodyIndex :type Integer) (bFromSweep :type Boolean) (SweepResult :type Hit%20Result%20Structure%20%28by%20ref%29))
  (bind (_bblockinghit _binitialoverlap _time _distance _location _impactpoint _normal _impactnormal) (Collision|BreakHitResult SweepResult))
  (bind _returnvalue (Transformation|GetActorLocation))
  (bind (_asphysics_ball_master _bsuccess) (Utilities|Casting|CastToPhysicsBall_Master OtherActor))
  (bind (_aspawn _bsuccess_2) (Utilities|Casting|CastToPawn OtherActor))
  (if (or (not (Variables|Default|GetServerAuthoritative)) (Networking|HasAuthority))
    (Variables|Default|SetOtherVelocity (Transformation|GetVelocity OtherActor))
    (bind _output_get (Variables|Default|SetOtherLocation (Transformation|GetActorLocation OtherActor)))
    (Variables|Default|SetHitOtherBall _bsuccess)
    (bind _output_get_2 (Variables|Default|SetHitOtherPawn _bsuccess_2))
    (if (and _output_get_2 (not (Variables|Default|GetbIsAttachedToPlayer)))
      (CallFunction|OnPawnTouchedMe _aspawn _impactnormal (Variables|Default|GetOtherVelocity))
      (else
        _))
    (bind _output_get_3 (Variables|FortPhysicsBall|SetbApplyingLift true))
    (Variables|FortPhysicsBall|SetbApplyingBackspin _output_get_3)
    (CallFunction|ApplyImpulsetoBall _output_get (Variables|Default|GetOtherVelocity) (* (Math|Vector|VectorLength (Variables|Default|GetOtherVelocity)) (Math|Float|SelectFloat 1.0 (Variables|Config|GetVelocityHitMagMultiplier) (Variables|Default|GetHitOtherBall))))
    (CallFunction|SetTickingEnabled true)
    (Networking|CanExecuteCosmeticEvents
      (:True
        (Audio|OnCollision (Variables|Default|GetFortCollisionAudio) _returnvalue (Variables|Default|GetOtherVelocity)))
      (:False
        _))
    (CallFunction|ImpactFX OtherActor _returnvalue true)
    (else
      (Networking|CanExecuteCosmeticEvents
        (:True
          (CallFunction|HandleBallCollisionAudio (Math|Vector|VectorLength (* _impactnormal (Variables|/Script/Engine.MovementComponent|GetVelocity (Variables|Default|GetProjectileMovement))))))))))

(event Custom|OnPawnTouchedMe
  (params (TouchingPawn :type Pawn%20Object%20Reference) (ImpactNormal :type Vector) (OtherVelocity :type Vector))
  (CallFunction|PlayPawnImpactAudio TouchingPawn ImpactNormal OtherVelocity))

(event Collision|EventHit
  (params (MyComp :type Primitive%20Component%20Object%20Reference) (Other :type Actor%20Object%20Reference) (OtherComp :type Primitive%20Component%20Object%20Reference) (bSelfMoved :type Boolean) (HitLocation :type Vector :default "0, 0, 0") (HitNormal :type Vector :default "0, 0, 0") (NormalImpulse :type Vector :default "0, 0, 0") (Hit :type Hit%20Result%20Structure%20%28by%20ref%29))
  (bind (_bblockinghit _binitialoverlap _time _distance _location _impactpoint _normal _impactnormal) (Collision|BreakHitResult Hit))
  (bind _projectilemovement (Variables|Default|GetProjectileMovement))
  (bind _velocity (Variables|/Script/Engine.MovementComponent|GetVelocity _projectilemovement))
  (bind _returnvalue (* _impactnormal _velocity))
  (Audio|OnCollision (Variables|Default|GetFortCollisionAudio) (Transformation|GetActorLocation) _returnvalue)
  (bind _output_get (Variables|FortPhysicsBall|SetbApplyingLift false))
  (Variables|FortPhysicsBall|SetbApplyingBackspin _output_get)
  (Variables|Default|SetHitEventMagnitude (Math|Vector|VectorLength _returnvalue))
  (Networking|CanExecuteCosmeticEvents
    (:True
      (CallFunction|ImpactFX Other))))

(event Custom|Water%20Collision
  (params (WaterBody :type Fort%20Water%20Body%20Actor%20Object%20Reference) (WaterInteractionComponent :type Fort%20Water%20Interaction%20Component%20Object%20Reference) (bIsFirstBody :type Boolean))
  (bind (_waterplanelocation _waterplanenormal _watersurfaceposition) (FX|GetLastWaterSurfaceInfo WaterInteractionComponent :!class /Script/FortniteGame.FortWaterInteractionComponent))
  ;[1 1 1] New Water System Impacts
  (bind _returnvalue (Effects|SpawnEmitteratLocation "/Game/Abilities/Toys/Shared/P_Toy_Impact_Water.P_Toy_Impact_Water" _watersurfaceposition))
  (Variables|Default|SetWaterImpactFX _returnvalue))

(event OnEnterWaterEvent(WaterInteractionComponent)
  (params (WaterBody :type Fort%20Water%20Body%20Actor%20Object%20Reference) (WaterInteractionComponent :type Fort%20Water%20Interaction%20Component%20Object%20Reference) (bIsFirstBody :type Boolean))
  (Audio|PlaySoundatLocation (Variables|Sound|GetWaterImpactSound) (Transformation|GetActorLocation)))

(event EventEnteredAirVent
  (params (LaunchVelocity :type Float%20%28double-precision%29) (BaseForceScale :type Float%20%28double-precision%29) (Direction :type Vector :default "0, 0, 0"))
  ;[1 0.1851 0.9664] Based on logic from B_HVAC_Prop
  (Variables|/Script/Engine.MovementComponent|SetVelocity (Math|Vector|Lerp(Vector) (* Direction (* LaunchVelocity BaseForceScale)) (Variables|/Script/Engine.MovementComponent|GetVelocity (Variables|Default|GetProjectileMovement)) 0.5) (Variables|Default|GetProjectileMovement)))

(event Custom|StartMovement
  (params (Initial%20Velocity%20Magnitude :type Float%20%28double-precision%29))
  (bind _output_get (Variables|FortPhysicsBall|SetbApplyingLift true))
  (Variables|FortPhysicsBall|SetbApplyingBackspin _output_get)
  (CallFunction|ApplyImpulsetoBall "0, 0, 0" "0.000000,15.000000,15.000000" (* (Math|Vector|VectorLength "10.000000,0.000000,10.000000") (Math|Float|SelectFloat 1.0 Initial%20Velocity%20Magnitude (Variables|Default|GetHitOtherBall))))
  (CallFunction|SetTickingEnabled true)
  (Networking|CanExecuteCosmeticEvents
    (:True
      (Audio|OnCollision (Variables|Default|GetFortCollisionAudio) (Transformation|GetActorLocation) (0.0 0.0 1e+01)))))

(fn ApplyImpulseToBall
  (params (HitLocation :type Vector) (HitVelocity :type Vector) (HitMagnitude :type Float%20%28double-precision%29))
  (if (Math|Boolean|ORBoolean (> HitMagnitude 0.0) (Variables|Default|GetHitOtherBall) (Variables|Default|GetHitOtherPawn))
    (bind _returnvalue (CallFunction|GetHitScale (Variables|GetHitLocation) (Variables|GetHitVelocity)))
    (Variables|/Script/Engine.MovementComponent|SetVelocity (Math|Vector|ClampVectorSize (+ (* (Transformation|GetVelocity) (Math|Float|SelectFloat (Math|Float|Square (Variables|/Script/Engine.ProjectileMovementComponent|GetBounciness (Variables|Default|GetProjectileMovement))) (* (Variables|/Script/Engine.ProjectileMovementComponent|GetBounciness (Variables|Default|GetProjectileMovement)) -1.0) (Variables|Default|GetHitOtherBall))) (* (* (Math|Vector|GetForwardVector (Math|Rotator|MakeRotator 0.0 (Variables|Config|GetHitPitch) (.yaw (Math|Conversions|RotationFromXVector (- (Transformation|GetActorLocation) HitLocation))))) (Math|Float|Min(Float) HitMagnitude (Variables|Config|GetHitMagMax))) _returnvalue)) 0.0 (Variables|Config|GetHitMagMax)) (Variables|Default|GetProjectileMovement))))

(fn CheckStopInWaterOnServer
  (bind (_velocity_x _velocity_y _velocity_z) (Math|Vector|BreakVector (Variables|/Script/Engine.MovementComponent|GetVelocity_X (Variables|Default|GetProjectileMovement))))
  (Networking|CanExecuteCosmeticEvents
    (:False
      (if (Water|IsinWater (Variables|FortPhysicsBall|GetWaterInteractionComponent))
        (if (and (< (Math|Float|Absolute(Float) _velocity_z) (Variables|Default|GetStopInWaterVelocityZ)) (< (Math|Vector|VectorLengthXYSquared (Variables|/Script/Engine.MovementComponent|GetVelocity (Variables|Default|GetProjectileMovement))) (Math|Float|Square (Variables|Default|GetStopInWaterVelocityHorizontal))))
          (Game|Components|ProjectileMovement|StopSimulating (Variables|Default|GetProjectileMovement) (Collision|MakeHitResult false)))))))

(fn GetHitScale
  (params (HitLocation :type Vector) (HitVelocity :type Vector))
  (return (Math|Float|Clamp(Float) (Math|Vector|DotProduct (Math|Vector|Normalize (* (- (Transformation|GetActorLocation) HitLocation) (Math|Vector|MakeVector 1.0 1.0))) (Math|Vector|Normalize (Variables|GetHitVelocity))) (Variables|Default|GetMinAngularPower))))

(fn HandleBallCollisionAudio
  (params (HitMagnitude :type Float%20%28single-precision%29))
  (bind _collisionaudio (Variables|Default|GetCollisionAudio))
  (Audio|Parameter|SetFloatParameter _collisionaudio "Movement.BallHitMagnitude" HitMagnitude :!class /Script/AudioExtensions.AudioParameterControllerInterface)
  (if (Audio|Components|Audio|IsPlaying _collisionaudio)
    (Audio|Parameter|ExecuteTriggerParameter _collisionaudio "Trigger.BallHit")
    (else
      (Audio|Components|Audio|Play _collisionaudio 0.0))))

(fn ImpactFX
  (params (HitActor :type Actor%20Object%20Reference) (SpawnLocation :type Vector) (Overlap? :type Boolean))
  ;[1 1 1] Check bounce count and switch based on if water volume overlapped
  (bind _returnvalue (Game|GetCurrentLevelName true))
  (if (Utilities|String|Equal,CaseInsensitive(String) _returnvalue "Frontend")
    _
    (elif (>= (Variables|Default|GetBounceCount) (select Overlap? 6 3))
      _
      (else
        (Utilities|Casting|CastToFortWaterBodyActor HitActor
          (:CastFailed
            ;[1 1 1] Traces for Water volume or world
            (bind _outhit (Collision|SphereTraceByChannel (+ (Transformation|GetWorldLocation (Variables|FortPhysicsBall|GetSphereCollision)) "0, 0,65.000000") (+ (Transformation|GetWorldLocation (Variables|FortPhysicsBall|GetSphereCollision)) "0, 0,-20.000000") 8.0 "TraceTypeQuery13" true))
            (LocalVariables|SetSpawnLoc (.ImpactPoint (Collision|BreakHitResult _outhit)))
            (if (Utilities|Equal(Object) (.PhysMat (Collision|BreakHitResult _outhit)) "/Game/FortressPhysicalMaterials/Water.Water")
              (bind _returnvalue_2 (Effects|SpawnEmitteratLocation "/Game/Abilities/Toys/Shared/P_Toy_Impact_Water.P_Toy_Impact_Water" (LocalVariables|GetSpawnLoc)))
              (Variables|Default|SetWaterImpactFX _returnvalue_2)
              (else
                (bind _returnvalue_3 (Effects|SpawnEmitteratLocation "/Game/Abilities/Toys/Shared/P_Toy_Impact_Generic.P_Toy_Impact_Generic" (LocalVariables|GetSpawnLoc) "90.000000,0.000000, 0"))
                (bind _output_get (Variables|Default|SetGenericImpactFX _returnvalue_3))
                ;[1 1 1] Setup Particle Paramters
                (Effects|Components|ParticleSystem|SetFloatParameter _output_get "SmokeRadius" (Variables|Default|GetDustRadius))
                (Effects|Components|ParticleSystem|SetVectorParameter (Variables|Default|GetGenericImpactFX) "DustSize" (Variables|Default|GetDustSize))))
            ;[1 1 1] Increment overlap/bounce count
            (Variables|Default|SetBounceCount (+ (Variables|Default|GetBounceCount) 1))))))))

(fn MakeBallDormant
  (Collision|SetActorEnableCollision false)
  (CallFunction|SetTickingEnabled false))

(fn OnRep_HitEventMagnitude
  (CallFunction|HandleBallCollisionAudio (Variables|Default|GetHitEventMagnitude)))

(fn OnRep_PlotTrajectory
  (Variables|Default|SetDebug_LastPosition (Transformation|GetActorLocation))
  (Development|PrintString (Utilities|String|Append "Start plotting, loc is " (Utilities|String|ToString(Vector) (Transformation|GetActorLocation)))))

(fn PlayPawnImpactAudio
  (params (TouchingPawn :type Pawn%20Object%20Reference) (ImpactNormal :type Vector) (PawnVelocity :type Vector))
  (bind _collisionaudio (Variables|Default|GetCollisionAudio))
  (Utilities|Casting|CastToFortPawn TouchingPawn)
  (if (> (Math|Vector|VectorLength PawnVelocity) 1e+01)
    (if (Audio|Components|Audio|IsPlaying _collisionaudio)
      (Audio|Parameter|ExecuteTriggerParameter _collisionaudio "Trigger.PlayerHit")
      (else
        (Audio|Components|Audio|Play _collisionaudio 0.0)))))

(fn SetBallVelocityAndStartTicking
  (params (NewVelocity :type Vector))
  (bind _output_get (Variables|/Script/Engine.MovementComponent|SetVelocity NewVelocity (Variables|Default|GetProjectileMovement)))
  (bind _output_get_2 (LocalVariables|SetLocalIsTicking (Math|Vector|NotEqual(Vector) _output_get "0, 0, 0" 0.0001)))
  (CallFunction|SetTickingEnabled _output_get_2)
  (Variables|Default|SetEverCameToRest false)
  (bind _output_get_3 (Variables|FortPhysicsBall|SetbApplyingLift (LocalVariables|GetLocalIsTicking)))
  (Variables|FortPhysicsBall|SetbApplyingBackspin _output_get_3))

(fn SetIsAttachedToPlayer
  (params (IsAttached :type Boolean))
  (Variables|Default|SetbIsAttachedToPlayer IsAttached))

(fn SetTickingEnabled
  (params (ShouldTick :type Boolean))
  (bind _projectilemovement (Variables|Default|GetProjectileMovement))
  (Utilities|FlowControl|Sequence
    (:then_0
      (Components|Tick|SetComponentTickEnabled (Variables|Default|GetProjectileMovement) ShouldTick)
      (if ShouldTick
        (Components|Movement|SetUpdatedComponent (Variables|Default|GetProjectileMovement) (Variables|FortPhysicsBall|GetSphereCollision))
        (else
          (bind _output_get (Variables|FortPhysicsBall|SetbApplyingLift false))
          (Variables|FortPhysicsBall|SetbApplyingBackspin _output_get)))
      (Networking|CanExecuteCosmeticEvents
        (:True
          (Actor|Tick|SetActorTickEnabled ShouldTick)
          (if ShouldTick
            (Game|Components|ProjectileMovement|Interpolation|SetInterpolatedComponent _projectilemovement (Variables|Default|GetMesh))
            (Variables|/Script/Engine.ProjectileMovementComponent|SetbInterpMovement true _projectilemovement)
            (else
              (Audio|Components|Audio|FadeOut (Variables|Default|GetRollingAudio) 0.1 0.0 "Linear" :!class /Script/Engine.AudioComponent)
              (Audio|Components|Audio|FadeOut (Variables|Default|GetCollisionAudio) 0.1 0.0 "Linear" :!class /Script/Engine.AudioComponent)
              (Variables|FortPhysicsBall|SetbRollingAudioFaded true))))
        (:False
          (Actor|Tick|SetActorTickEnabled (and ShouldTick (Variables|FortPhysicsBall|GetbEnableLiftAndDrag))))))
    (:then_1
      (Networking|SwitchHasAuthority
        (:Authority
          (Networking|SetNetDormancy (select ShouldTick "DORM_Awake" "DORM_DormantAll"))
          (Networking|FlushNetDormancy))))))

(fn UpdateDirectionAndRotationRate
  (params (DeltaSeconds :type Float%20%28double-precision%29))
  (FortPhysicsBall|UpdateDirectionandRotation DeltaSeconds)
  (return))

(fn ConstructionScript)
